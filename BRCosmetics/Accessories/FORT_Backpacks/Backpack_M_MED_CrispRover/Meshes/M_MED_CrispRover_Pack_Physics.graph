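# Physics asset for the CrispRover backpack cosmetic mesh, providing bodies for its arrow bones.
# A longer capsule on dyn_arrow and a small capsule on dyn_arrow_base form the two bodies.
# A single constraint joins them, with zero twist and swing-1 limits and a wide swing-2 limit, suggesting a hinge-like swing on one axis.
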PhysicsAsset /BRCosmetics/Accessories/FORT_Backpacks/Backpack_M_MED_CrispRover/Meshes/M_MED_CrispRover_Pack_Physics
bodies {
 dyn_arrow [{shapeType=Capsule, length=4.6422, radius=1.2321, shapeName=dyn_arrow_capsule, center=<0 0.0141 0>, rotation=<rot 0 -90 -1.6843>}]
 dyn_arrow_base [{shapeType=Capsule, length=0.505, radius=0.505, shapeName=dyn_arrow_base_capsule, rotation=<rot 0 -90 0>}]
}
constraints [{bone1Name=dyn_arrow, bone2Name=dyn_arrow_base, swing1LimitDegrees=0, swing2LimitDegrees=104.96, twistLimitDegrees=0}]
